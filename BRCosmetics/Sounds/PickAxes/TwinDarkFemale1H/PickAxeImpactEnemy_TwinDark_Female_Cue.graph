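# Sound cue for the female Twin Dark one-handed pickaxe hitting an enemy, using pickaxe impact attenuation, a melee weapon sound class and player melee concurrency.
# A quality-level root picks a branch: six random impact waves, three random waves, or one fixed wave shared by the last two levels.
# Each branch runs through a short fixed delay and a modulator, and the cue primes on load.
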SoundCue /BRCosmetics/Sounds/PickAxes/TwinDarkFemale1H/PickAxeImpactEnemy_TwinDark_Female_Cue
root=QualityLevel_0
nodes [
 {id=Delay_1, children [Modulator_1], properties {delayMax=0.05, delayMin=0.05}}
 {id=Delay_2, children [Modulator_2], properties {delayMax=0.05, delayMin=0.05}}
 {id=Delay_3, children [Modulator_3], properties {delayMax=0.05, delayMin=0.05}}
 {id=Modulator_1, children [Random_0]}
 {id=Modulator_2, children [Random_1]}
 {id=Modulator_3, children [WavePlayer_0]}
 {id=QualityLevel_0, children [Delay_1,Delay_2,Delay_3,Delay_3]}
 {
  id=Random_0
  children [WavePlayer_6, WavePlayer_7, WavePlayer_8, WavePlayer_9, WavePlayer_10, WavePlayer_11]
  properties {weights [1,1,1,1,1,1]}
 }
 {id=Random_1, children [WavePlayer_41, WavePlayer_42, WavePlayer_43], properties {weights [1,1,1]}}
 {id=WavePlayer_0, properties {soundWaveAssetPtr=&/PickAxe_TwinDark_Enemy_Impact_01}}
 {id=WavePlayer_10, properties {soundWaveAssetPtr=&/PickAxe_TwinDark_Enemy_Impact_05}}
 {id=WavePlayer_11, properties {soundWaveAssetPtr=&/PickAxe_TwinDark_Enemy_Impact_06}}
 {id=WavePlayer_41, properties {soundWaveAssetPtr=&PickAxe_TwinDark_Enemy_Impact_01}}
 {id=WavePlayer_42, properties {soundWaveAssetPtr=&/PickAxe_TwinDark_Enemy_Impact_03}}
 {id=WavePlayer_43, properties {soundWaveAssetPtr=&/PickAxe_TwinDark_Enemy_Impact_02}}
 {id=WavePlayer_6, properties {soundWaveAssetPtr=&PickAxe_TwinDark_Enemy_Impact_01}}
 {id=WavePlayer_7, properties {soundWaveAssetPtr=&PickAxe_TwinDark_Enemy_Impact_02}}
 {id=WavePlayer_8, properties {soundWaveAssetPtr=&PickAxe_TwinDark_Enemy_Impact_03}}
 {id=WavePlayer_9, properties {soundWaveAssetPtr=&/PickAxe_TwinDark_Enemy_Impact_04}}
]
properties {
 attenuationSettings=/Game/Sounds/Attenuation/Fort_Pickaxe_Impact_Att
 audioPropertiesSheet=/BRCosmetics/Sounds/Settings/PropertySheets/Pickaxe/APS_Cosmetic_Pickaxe_Impact
 bPrimeOnLoad=true
 maxDistance=4300
 soundClassObject=&Fort_WeaponMelee_PSM
 volumeMultiplier=1.5
 concurrencySet [&Player_Melee_Weapon_Conc]
}
