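# Sound cue for the ready sound of the Twin Dark female one-handed pickaxe cosmetic.
# A quality switch sits at the root, picking randomly among three waves at the first level and two at the second, while the last two levels share one fixed wave.
# Each branch passes through a default modulator.
# It uses close pickaxe attenuation and the melee weapon sound class, primes on load and overrides concurrency to block new instances at the limit.
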SoundCue /BRCosmetics/Sounds/PickAxes/TwinDarkFemale1H/PickAxeReady_TwinDark_Female_Cue
root=QualityLevel_0
nodes [
 {id=Modulator_1, children [Random_0]}
 {id=Modulator_2, children [Random_3]}
 {id=Modulator_3, children [WavePlayer_5]}
 {id=QualityLevel_0, children [Modulator_1,Modulator_2,Modulator_3,Modulator_3]}
 {id=Random_0, children [WavePlayer_0,WavePlayer_1,WavePlayer_2], properties {weights [1,1,1]}}
 {id=Random_3, children [WavePlayer_3,WavePlayer_4], properties {weights [1,1]}}
 {id=WavePlayer_0, properties {soundWaveAssetPtr=&/PickAxe_Twin_Dark_Axe_Ready_03}}
 {id=WavePlayer_1, properties {soundWaveAssetPtr=&/PickAxe_Twin_Dark_Axe_Ready_01}}
 {id=WavePlayer_2, properties {soundWaveAssetPtr=&/PickAxe_Twin_Dark_Axe_Ready_02}}
 {id=WavePlayer_3, properties {soundWaveAssetPtr=&PickAxe_Twin_Dark_Axe_Ready_01}}
 {id=WavePlayer_4, properties {soundWaveAssetPtr=&PickAxe_Twin_Dark_Axe_Ready_02}}
 {id=WavePlayer_5, properties {soundWaveAssetPtr=&PickAxe_Twin_Dark_Axe_Ready_01}}
]
properties {
 attenuationSettings=/Game/Sounds/Attenuation/Fort_Weapon_Pickaxe_Close_Att
 audioPropertiesSheet=/BRCosmetics/Sounds/Settings/PropertySheets/Pickaxe/APS_Cosmetic_Pickaxe_Ready
 bOverrideConcurrency=true
 bPrimeOnLoad=true
 maxDistance=1536
 soundClassObject=&Fort_WeaponMelee
 volumeMultiplier=0.176
 concurrencyOverrides {resolutionRule=PreventNew}
}
