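# Sound cue for the swing of the TwinDark Female one-handed cosmetic pickaxe.
# A mixer layers two modulated branches, one under close attenuation with TwinDark swing waves and one under medium attenuation with shared distant fire-axe swing waves.
# In each branch a quality-level node picks between a six-variant random set, a three-variant random set, or a single wave, trimming variation at lower quality.
# It primes on load and uses player melee weapon concurrency.
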SoundCue /BRCosmetics/Sounds/PickAxes/TwinDarkFemale1H/PickAxeSwing_TwinDark_Female_Cue
root=Mixer_0
nodes [
 {id=Attenuation_0, children [Modulator_1], properties {attenuationSettings=/Game/Sounds/Attenuation/Fort_Weapon_Pickaxe_Close_Att}}
 {id=Attenuation_1, children [Modulator_3], properties {attenuationSettings=/Game/Sounds/Attenuation/Fort_Weapon_Pickaxe_Medium_Att}}
 {id=Mixer_0, children [Attenuation_0, Attenuation_1], properties {inputVolume [1,1]}}
 {id=Modulator_1, children [QualityLevel_0]}
 {id=Modulator_3, children [QualityLevel_2]}
 {id=QualityLevel_0, children [Random_0,Random_3,WavePlayer_6,WavePlayer_6]}
 {id=QualityLevel_2, children [Random_4, Random_5, WavePlayer_29, WavePlayer_29]}
 {id=Random_0, children [WavePlayer_0,WavePlayer_1,WavePlayer_2,WavePlayer_3,WavePlayer_4,WavePlayer_5], properties {weights [1,1,1,1,1,1]}}
 {id=Random_3, children [WavePlayer_18, WavePlayer_19, WavePlayer_20], properties {weights [1,1,1]}}
 {
  id=Random_4
  children [WavePlayer_17, WavePlayer_21, WavePlayer_22, WavePlayer_23, WavePlayer_24, WavePlayer_25]
  properties {weights [1,1,1,1,1,1]}
 }
 {id=Random_5, children [WavePlayer_27, WavePlayer_26, WavePlayer_28], properties {weights [1,1,1]}}
 {id=WavePlayer_0, properties {soundWaveAssetPtr=&/PickAxe_TwinDark_Swing_01}}
 {id=WavePlayer_1, properties {soundWaveAssetPtr=&/PickAxe_TwinDark_Swing_02}}
 {id=WavePlayer_17, properties {soundWaveAssetPtr=/Game/Athena/Sounds/Weapons/Fire_Axe_Swing_Distant_04}}
 {id=WavePlayer_18, properties {soundWaveAssetPtr=&PickAxe_TwinDark_Swing_01}}
 {id=WavePlayer_19, properties {soundWaveAssetPtr=&/PickAxe_TwinDark_Swing_03}}
 {id=WavePlayer_2, properties {soundWaveAssetPtr=&PickAxe_TwinDark_Swing_03}}
 {id=WavePlayer_20, properties {soundWaveAssetPtr=&PickAxe_TwinDark_Swing_02}}
 {id=WavePlayer_21, properties {soundWaveAssetPtr=/Game/Athena/Sounds/Weapons/Fire_Axe_Swing_Distant_01}}
 {id=WavePlayer_22, properties {soundWaveAssetPtr=/Game/Athena/Sounds/Weapons/Fire_Axe_Swing_Distant_02}}
 {id=WavePlayer_23, properties {soundWaveAssetPtr=/Game/Athena/Sounds/Weapons/Fire_Axe_Swing_Distant_03}}
 {id=WavePlayer_24, properties {soundWaveAssetPtr=/Game/Athena/Sounds/Weapons/Fire_Axe_Swing_Distant_05}}
 {id=WavePlayer_25, properties {soundWaveAssetPtr=/Game/Athena/Sounds/Weapons/Fire_Axe_Swing_Distant_06}}
 {id=WavePlayer_26, properties {soundWaveAssetPtr=&Fire_Axe_Swing_Distant_03}}
 {id=WavePlayer_27, properties {soundWaveAssetPtr=&Fire_Axe_Swing_Distant_04}}
 {id=WavePlayer_28, properties {soundWaveAssetPtr=&Fire_Axe_Swing_Distant_06}}
 {id=WavePlayer_29, properties {soundWaveAssetPtr=&Fire_Axe_Swing_Distant_06}}
 {id=WavePlayer_3, properties {soundWaveAssetPtr=&/PickAxe_TwinDark_Swing_04}}
 {id=WavePlayer_4, properties {soundWaveAssetPtr=&/PickAxe_TwinDark_Swing_05}}
 {id=WavePlayer_5, properties {soundWaveAssetPtr=&/PickAxe_TwinDark_Swing_06}}
 {id=WavePlayer_6, properties {soundWaveAssetPtr=&PickAxe_TwinDark_Swing_01}}
]
properties {
 audioPropertiesSheet=/BRCosmetics/Sounds/Settings/PropertySheets/Pickaxe/APS_Cosmetic_Pickaxe_Swing
 bPrimeOnLoad=true
 maxDistance=3000
 soundClassObject=&Fort_WeaponMelee_PSM
 volumeMultiplier=0.6
 concurrencySet [&Player_Melee_Weapon_Conc]
}
